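; Reactive cosmetic audio component that plays Sound Library sounds when a watched stat changes on the player pawn of its owning Athena part modifier.
; Once the game state is ready, it watches the configured stat tag, except DBNO, which binds to the pawn's downed event instead.
; Outside frontend display, positive values play a one-per-owner sound.
; It can also push the value as an integer parameter to stored audio components.
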(event EventBeginPlay
  (bind _partsmodifier (Variables|Private|GetPartsModifier))
  (bind _returnvalue (CustomCharacter|GetAssociatedPlayerPawn _partsmodifier))
  (Class|FortAsyncActionGameStateReady|GameStateReadyAsync
    (:OnReady
      (bind _asb_athena_part_modifier_generic (Utilities|Casting|CastToB_Athena_PartModifier_Generic (Components|GetOwner)))
      (Variables|Private|SetPartsModifier _asb_athena_part_modifier_generic) ; removed from private to access\nteam reactivity
      (if (GameplayTags|HasTag (GameplayTags|MakeLiteralGameplayTagContainer "(GameplayTags=((TagName=\"StatManager.DBNO\")))") (Variables|StatWatcher|GetWatcherStatTag) true)
        (Utilities|IsValid _returnvalue
          (:Is%20Valid
            (EventDispatchers|AssignOnPlayerStartDBNO _returnvalue (delegate Custom|PlayerStartDBNO))
            (CallFunction|FinishedBeginPlay)))
        (else
          (bind (_statname _statvalue) (Cosmetics|WatchCosmeticStat (CustomCharacter|GetAssociatedPlayerPawn (Variables|Private|GetPartsModifier)) (Variables|StatWatcher|GetWatcherStatTag) (Variables|StatWatcher|GetDefault%20Value) (Variables|StatWatcher|GetBroadcast%20Initial%20Value) (Variables|StatWatcher|GetInitial%20Value%20Broadcast%20Delay) (Class|FortCosmeticFunctionLibrary|GetCosmeticStatDelay (Variables|StatWatcher|GetWatcherStatTag)) (Variables|StatWatcher|GetConsume%20Waiting%20Notifications))
            (:then
              (CallFunction|FinishedBeginPlay))
            (:OnStatChanged
              (CallFunction|AudioStatWatcher _statname _statvalue (Variables|StatWatcher|GetSoundLibraryTag)))))))))

(event Custom|PlayerStartDBNO
  (CallFunction|AudioStatWatcher (Variables|Private|GetStatValue) 1 (Variables|StatWatcher|GetSoundLibraryTag)))

(event Custom|AudioStatWatcher
  (params (StatName :type Name) (StatValue :type Integer) (LibraryTag :type Sound%20Library%20Tag%20Structure%20%28by%20ref%29))
  (bind _partsmodifier (Variables|Private|GetPartsModifier)
    (:then
      (if (not (Reactive|Frontend|IsCosmeticObjectDisplayedinFrontendContext (CustomCharacter|GetSkeletalMeshForPartOnPlayerPawn _partsmodifier)))
        (CallFunction|PlayLibrarySound StatName StatValue LibraryTag)
        (if (and (Variables|Private|GetSuccess) (Variables|StatWatcher|GetUpdateAudioParameter))
          (CallFunction|UpdateAudioStatValue (Variables|Private|GetStatValue) StatValue))))
    (:else)))

(event Custom|PlayLibrarySound
  (params (StatName :type Name) (StatValue :type Integer) (LibraryTag :type Sound%20Library%20Tag%20Structure%20%28by%20ref%29))
  (bind _partsmodifier (Variables|Private|GetPartsModifier))
  (bind _returnvalue (CustomCharacter|GetAssociatedPlayerPawn _partsmodifier))
  (if (> StatValue 0)
    (Utilities|IsValid _returnvalue
      (:Is%20Valid
        (bind _returnvalue_2 (SoundLibrary|PlaySoundSimple (WorldSubsystems|GetSoundLibrarySubsystem) _returnvalue (Utilities|Struct|MakeSimpleSoundSettings LibraryTag true true true false 1.0 1.0 0.0 "/Game/Sounds/Concurrency_Rules/OneMaximum_Owner_StopOldest.OneMaximum_Owner_StopOldest" 0 "None" (CustomCharacter|GetSkeletalMeshForPartOnPlayerPawn _partsmodifier)) :!class /Script/SoundLibrary.SoundLibrarySubsystem))
        (if (Utilities|Array|IsValidIndex (.AudioComponents (Utilities|Struct|BreakSoundLibraryPlayResult _returnvalue_2)))
          (Variables|Private|SetSuccess (.bSuccess (Utilities|Struct|BreakSoundLibraryPlayResult _returnvalue_2)))
          (Variables|Private|SetAudioComponents (.AudioComponents (Utilities|Struct|BreakSoundLibraryPlayResult _returnvalue_2))))))))

(event Custom|UpdateAudioStatValue
  (params (InName :type Name) (StatValue :type Integer))
  (for _array_element (Variables|Private|GetAudioComponents)
    (Audio|Parameter|SetIntegerParameter _array_element InName StatValue :!class /Script/AudioExtensions.AudioParameterControllerInterface)))

(event Custom|FinishedBeginPlay)
