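; Manager Blueprint for teammate-reactive cosmetic audio, playing sounds on groups of player pawns that share a reactive group tag.
; On BeginPlay it listens on the GameState's event router for init and reaction gameplay events.
; Init events add or remove pawns in a tag-keyed map of groups, dropping emptied groups.
; Reaction events play a Sound Library tag on every member of the matching group, with no instigator or group-size filter.
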(event EventBeginPlay
  (bind _returnvalue (GameplayEventRouter|GetEventRouter "/Script/FortniteGame.FortGameState" (Game|GetGameState)))
  (Utilities|FlowControl|Sequence
    (:then_0
      ; Init values from first teammate to set up team reaction (maybe each member buys in...?)
      (bind (_asynctaskproxy _payload) (GameplayEventRouter|StartListeningToEvent _returnvalue "/Game/Sounds/Fort_Audio_Blueprints/Cosmetics/GameplayEvent_TeammateReactive_Init.GameplayEvent_TeammateReactive_Init")
        (:OnEventReceived
          (if (.AddPlayer_51_32CD1CC64F469BE3BC70CDAD5FF7FE71 (Utilities|Struct|BreakGameplayEventTeammateReactiveInit _payload))
            (CallFunction|AddPlayertoGroup (Utilities|Struct|BreakGameplayEventTeammateReactiveInit _payload) (.ReactiveGroupTag_42_95207CAF44000EA766F42AB4CE6B0938 (Utilities|Struct|BreakGameplayEventTeammateReactiveInit _payload)))
            (else
              (CallFunction|RemovePlayerfromGroup (Utilities|Struct|BreakGameplayEventTeammateReactiveInit _payload) (.ReactiveGroupTag_42_95207CAF44000EA766F42AB4CE6B0938 (Utilities|Struct|BreakGameplayEventTeammateReactiveInit _payload))))))))
    (:then_1
      ; Trigger audio for the non instigating players
      (bind (_asynctaskproxy_2 _payload_2) (GameplayEventRouter|StartListeningToEvent _returnvalue "/Game/Sounds/Fort_Audio_Blueprints/Cosmetics/GameplayEvent_TeammateReactive.GameplayEvent_TeammateReactive")
        (:OnEventReceived
          (CallFunction|TriggerGroupReaction (Utilities|Struct|BreakGameplayEventTeammateReactive _payload_2) (.SoundLibraryTag_34_F2FE57E045F5453EDB32249A061A8C22 (Utilities|Struct|BreakGameplayEventTeammateReactive _payload_2))))))))

(fn AddPlayerToGroup
  (params (Player :type Fort%20Player%20Pawn%20Object%20Reference) (ReactiveGroup :type Gameplay%20Tag%20Structure))
  (bind _reactive_squad (Variables|Default|GetReactive%20Squad))
  ; Check if the group is already existing
  (LocalVariables|SetvPlayer Player)
  (bind _output_get (LocalVariables|SetvGroup ReactiveGroup))
  (if (.ReturnValue (Utilities|Map|Find _reactive_squad _output_get))
    ; Add new actor to existing group
    (if (not (Utilities|Array|ContainsItem (Utilities|Struct|BreakReactiveTeammateGroup (Utilities|Map|Find _reactive_squad _output_get)) (LocalVariables|GetvPlayer)))
      (bind _output_get_2 (LocalVariables|SetvPlayers (Utilities|Struct|BreakReactiveTeammateGroup (Utilities|Map|Find _reactive_squad _output_get))))
      (Utilities|Array|Add _output_get_2 (LocalVariables|GetvPlayer))
      (Utilities|Map|Add (Variables|Default|GetReactive%20Squad) (LocalVariables|GetvGroup) (Utilities|Struct|MakeReactiveTeammateGroup (LocalVariables|GetvPlayers))))
    (else
      ; Add the new group
      (Utilities|Map|Add _reactive_squad (LocalVariables|GetvGroup) (Utilities|Struct|MakeReactiveTeammateGroup (Utilities|Array|MakeArray (LocalVariables|GetvPlayer)))))))

(fn RemovePlayerFromGroup
  (params (Player :type Fort%20Player%20Pawn%20Object%20Reference) (ReactiveGroup :type Gameplay%20Tag%20Structure))
  (bind _reactive_squad (Variables|Default|GetReactive%20Squad))
  (bind (_value _returnvalue) (Utilities|Map|Find _reactive_squad ReactiveGroup))
  ; Remove the player from the reactive group
  (LocalVariables|SetvReactiveGroup ReactiveGroup)
  (if _returnvalue
    (bind _output_get (LocalVariables|SetvPlayers (Utilities|Struct|BreakReactiveTeammateGroup _value)))
    (Utilities|Array|RemoveItem _output_get Player)
    ; If there are remaining players in the group, update teh map
    ; If not, remove the reactive group from teh squad
    (if (Utilities|Array|IsNotEmpty (LocalVariables|GetvPlayers))
      (Utilities|Map|Add (Variables|Default|GetReactive%20Squad) (LocalVariables|GetvReactiveGroup) (Utilities|Struct|MakeReactiveTeammateGroup (LocalVariables|GetvPlayers)))
      (else
        (Utilities|Map|Remove (Variables|Default|GetReactive%20Squad) (LocalVariables|GetvReactiveGroup))))))

(fn TriggerGroupReaction
  (params (ReactingGroup :type Gameplay%20Tag%20Structure) (SoundLibraryTag :type Sound%20Library%20Tag%20Structure))
  (bind _reactive_squad (Variables|Default|GetReactive%20Squad))
  (bind (_value _returnvalue) (Utilities|Map|Find _reactive_squad ReactingGroup))
  (LocalVariables|SetSlTag SoundLibraryTag)
  ; Allow playback if there is more than one player in the group
  (if _returnvalue
    ; trigger sound playback at all player in the group
    (for _array_element (Utilities|Struct|BreakReactiveTeammateGroup _value)
      (SoundLibrary|PlaySoundSimple (WorldSubsystems|GetSoundLibrarySubsystem) _array_element (Utilities|Struct|MakeSimpleSoundSettings (LocalVariables|GetSlTag) true false true false 1.0 1.0 0.0 0 0 "None" (Variables|/Script/Engine.Actor|GetRootComponent _array_element)) :!class /Script/SoundLibrary.SoundLibrarySubsystem))))
